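# Additive, two-sided unlit material for Niagara mesh particles in the Eclipse backpack effect, drawing a tinted glowing disc with a circular bite.
# A UV sphere mask and noise texture modulate the color, which is boosted to HDR.
# A second sphere mask, inverted and multiplied with the disc, carves the eclipse shape.
# Dynamic parameters drive intensity (via "Hue Shiuft"), cutout radius and position, with X_Position feeding both axes and Y_Position unused.
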Material /BRCosmetics/Accessories/FORT_Backpacks/Backpack_F_MED_Eclipse/FX/M_Eclipse_Mask
graph {
 properties {
  bUsedWithNiagaraMeshParticles=true
  blendMode=BLEND_Additive
  pixelDepthOffsetMode=PDOM_Legacy
  shadingModel=MSM_Unlit
  twoSided=true
  usedShadingModels=Unlit
 }
 nodes [
  n1 = VectorParameter() {defaultValue=<rgba 0.875 0.1686 0.8132 1>}
  n2 = SphereMask(
   A=TextureCoordinate()
   B=Constant() {r=0.5}
   Radius=ScalarParameter() {defaultValue=0.46}
   Hardness=ScalarParameter() {defaultValue=0.8, parameterName=1}
  ) {hardnessPercent=60}
  n3 = LinearInterpolate(
   A=Multiply(A=ScalarParameter() {defaultValue=0.7, parameterName=Param_2}, B=Multiply(A=@n1.RGB, B=@n2))
   B=@n1.RGB
   Alpha=TextureSample() {texture=&T_NoisePack_01}.R
  )
  n4 = ScalarParameter() {defaultValue=5, parameterName=Param_1}
  n5 = DynamicParameter() {defaultValue=<rgba 1 0 0 0.5>, paramNames [Hue Shiuft, X_Position, Y_Position, Radius]}
  MaterialFunctionCall(Emissive=MaterialFunctionCall(
    LDR Input=@n3
    HDR Tint=Clamp(None=@n3)
    HDR Intensity Multiplier=Multiply(A=@n4, B=@n5.Hue Shiuft)
    Derive HDR Power=ScalarParameter() {parameterName=Param_4}
   ) {materialFunction=&DeriveHDRfromLDR}.RemappedRange) {materialFunction=&MF_FX_EmissiveAdjustments}
  n6 = TextureCoordinate()
  n7 = AppendVector(A=@n5.X_Position, B=@n5.X_Position)
  n8 = ScalarParameter() {defaultValue=30, parameterName=Param_3}
  Multiply(A=@n2, B=OneMinus(None=SphereMask(A=@n6, B=@n7, Radius=@n5.Radius, Hardness=@n8) {hardnessPercent=50}))
 ]
}
