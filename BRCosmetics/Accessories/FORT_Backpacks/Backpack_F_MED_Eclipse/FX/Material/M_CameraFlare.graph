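# Additive, unlit, two-sided Niagara sprite and mesh particle material for a camera flare in the Eclipse backpack effects.
# A flare texture, layered with panning caustics and pixel noise, forms the mask, shaped by particle alpha, a dynamic-parameter sphere mask and depth fade.
# An oversaturated pink-to-cyan UV gradient tints it by particle color.
# The Glow parameter, zero by default, blends toward an HDR spiky halo flare with plain particle alpha.
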Material /BRCosmetics/Accessories/FORT_Backpacks/Backpack_F_MED_Eclipse/FX/Material/M_CameraFlare
graph {
 properties {
  bUsedWithNiagaraMeshParticles=true
  bUsedWithNiagaraSprites=true
  blendMode=BLEND_Additive
  pixelDepthOffsetMode=PDOM_Legacy
  shadingModel=MSM_Unlit
  twoSided=true
  usedShadingModels=Unlit
 }
 nodes [
  Constant() {r=1}
  n1 = ParticleColor()
  n2 = TextureSample(UVs=TextureCoordinate() {coordinateIndex=1}) {samplerType=SAMPLERTYPE_Grayscale, texture=/Game/Effects/Fort_Effects/Textures/T_EFX_CameraFlare_01}
  n3 = Panner(Coordinate=TextureCoordinate() {coordinateIndex=1, uTiling=4, vTiling=0.01}) {speedX=-0.2, speedY=0.2}
  n4 = Panner(Coordinate=TextureCoordinate() {coordinateIndex=1, uTiling=3, vTiling=0.05}) {speedX=0.15, speedY=-0.3}
  n5 = Multiply(
   A=Multiply(
    A=TextureSample(UVs=@n3) {texture=&T_ShieldCaustics}.R
    B=TextureSample(UVs=@n4) {texture=&T_ShieldCaustics}.R
   )
   B=ScalarParameter() {defaultValue=22}
  )
  n6 = TextureCoordinate() {coordinateIndex=1}
  n7 = Sine(None=Multiply(A=ComponentMask(None=@n6) {r=true}) {constB=0.5})
  n8 = ComponentMask(None=@n6) {g=true}
  n9 = Sine(None=Multiply(A=@n8) {constB=0.5})
  n10 = Multiply(A=Multiply(A=@n7, B=@n7), B=Multiply(A=Power(Base=@n9) {constExponent=4}, B=OneMinus(None=@n8)))
  n11 = DynamicParameter()
  n12 = Multiply(A=Multiply(A=@n2.R, B=Multiply(A=@n5, B=@n10)), B=@n11.R)
  n13 = TextureCoordinate() {coordinateIndex=1, uTiling=0.35, vTiling=4}
  n14 = TextureSample(UVs=Panner(Coordinate=MaterialFunctionCall(XY=@n13) {materialFunction=&Swizzle}.YX) {speedX=0.1, speedY=0.15}) {
   samplerType=SAMPLERTYPE_LinearGrayscale
   texture=/Game/UI/Frontend/ItemShop/Com/vfx/T_CommunityVoting_Activities_Pixels.CommunityVoting_Activities_Pixels
  }
  n15 = TextureCoordinate() {coordinateIndex=1, uTiling=0.5, vTiling=2}
  n16 = TextureSample(UVs=Panner(Coordinate=MaterialFunctionCall(XY=@n15) {materialFunction=&Swizzle}.YX) {speedX=-0.15, speedY=-0.3}) {
   samplerType=SAMPLERTYPE_LinearGrayscale
   texture=&CommunityVoting_Activities_Pixels
  }
  n17 = Multiply(A=@n2.R, B=Multiply(A=@n14.R, B=@n16.R))
  n18 = ScalarParameter() {defaultValue=55, parameterName=Param1}
  n19 = Multiply(A=@n1.A, B=Add(A=Add(A=@n2.R, B=@n12), B=Multiply(A=@n17, B=@n18)))
  n20 = SphereMask(
   A=MaterialFunctionCall(Float2=TextureCoordinate() {coordinateIndex=1}) {materialFunction=&BreakOutFloat2Components}.G
   B=Constant() {r=0.5}
   Radius=Multiply(A=Constant() {r=2}, B=@n11.G)
  ) {attenuationRadius=1, hardnessPercent=0}
  LinearInterpolate(
   A=DepthFade(Opacity=Multiply(A=@n19, B=@n20), FadeDistance=ScalarParameter() {defaultValue=10, parameterName=DepthFade})
   B=@n1.A
   Alpha=ScalarParameter() {parameterName=Glow}
  )
  n21 = ComponentMask(None=TextureCoordinate() {coordinateIndex=1}) {r=true}
  n22 = Desaturation(
   None=LinearInterpolate(
    A=Constant3Vector() {constant=<rgba 3 0.5625 2.1494 0>}
    B=Constant3Vector() {constant=<rgba 0.0234 1.334 1.5 0>}
    Alpha=Power(Base=@n21) {constExponent=1.5}
   )
   Fraction=Constant() {r=-1}
  )
  n23 = TextureSample(UVs=MaterialFunctionCall() {materialFunction=&VectorToRadialValue}.Radial Coordinates) {samplerSource=SSM_Clamp_WorldGroupSettings, texture=/Game/Effects/Fort_Effects/Textures/Grad/T_Halo_Corona_Gradient}
  n24 = Constant() {r=-0.75}
  n25 = TextureCoordinate()
  n26 = Constant() {r=0.05}
  n27 = TextureCoordinate()
  n28 = Constant() {r=0.05}
  n29 = MaterialFunctionCall(
   X=TextureSample(UVs=Rotator(Coordinate=@n25, Time=@n26) {centerX=1, centerY=0}) {texture=/Game/Effects/Fort_Effects/Textures/Flares/T_Flare_Spikeball}.R
   Y=TextureSample() {texture=&T_Flare_Spikeball}.G
   Z=TextureSample(UVs=Rotator(Coordinate=@n27, Time=@n28) {centerX=1, centerY=0, speed=-0.25}) {texture=&T_Flare_Spikeball}.B
  ) {materialFunction=&MakeFloat3}
  n30 = Add(
   A=Multiply(A=Desaturation(None=@n23.RGB, Fraction=@n24)) {constB=0.001}
   B=Power(Base=Desaturation(None=@n29.Result, Fraction=@n24), Exp=ScalarParameter() {defaultValue=2, parameterName=GlowPower})
  )
  n31 = MaterialFunctionCall(LDR Input=Clamp(None=@n30), HDR Tint=@n30, HDR Intensity Multiplier=ScalarParameter() {defaultValue=10, parameterName=HDR}) {materialFunction=&DeriveHDRfromLDR}
  MaterialFunctionCall(Emissive=LinearInterpolate(A=Multiply(A=@n22, B=@n1.RGB), B=Multiply(A=@n31.RemappedRange, B=@n1.RGB), Alpha=ScalarParameter() {parameterName=Glow})) {materialFunction=&MF_FX_EmissiveAdjustments}
  MaterialFunctionCall() {materialFunction=&DebugTimeSine}
  n32 = ScalarParameter() {defaultValue=3, parameterName=Param_1}
  MaterialFunctionCall(WidthBase=@n32, WidthTip=@n32) {materialFunction=&SplineThicken}
  Panner()
 ]
}
