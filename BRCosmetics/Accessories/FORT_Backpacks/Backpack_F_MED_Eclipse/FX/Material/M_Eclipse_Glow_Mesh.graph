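# Additive, unlit, two-sided glow material for Niagara mesh particles in the Eclipse backpack effect, with a Use_Trail parameter blending between glow and trail looks.
# Glow mode combines panning caustics, noise and a UV gradient with a color tint, using depth fades for opacity.
# Trail mode takes particle color with a dynamic mask along V.
# SplineThicken sets width through world position offset, with trail mode adding directional displacement and sine waves driven by dynamic parameters.
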Material /BRCosmetics/Accessories/FORT_Backpacks/Backpack_F_MED_Eclipse/FX/Material/M_Eclipse_Glow_Mesh
graph {
 properties {
  bUsedWithNiagaraMeshParticles=true
  blendMode=BLEND_Additive
  pixelDepthOffsetMode=PDOM_Legacy
  shadingModel=MSM_Unlit
  twoSided=true
  usedShadingModels=Unlit
 }
 nodes [
  Constant()
  n1 = OneMinus(None=ComponentMask(None=TextureCoordinate()) {r=true})
  n2 = ScalarParameter() {defaultValue=70, parameterName=Param_1}
  n3 = DepthFade(Opacity=ScalarParameter() {defaultValue=3, parameterName=Dep}) {fadeDistanceDefault=40}
  n4 = ParticleColor()
  n5 = ComponentMask(None=TextureCoordinate()) {g=true}
  n6 = Constant() {r=1}
  n7 = DynamicParameter() {paramNames [Mask,Spline,Param3,Param4]}
  LinearInterpolate(
   A=Multiply(A=DepthFade(Opacity=@n1, FadeDistance=@n2) {fadeDistanceDefault=40}, B=OneMinus(None=@n3))
   B=Multiply(A=@n4.A, B=SphereMask(A=@n5, B=@n6, Radius=@n7.Mask) {attenuationRadius=0.5})
   Alpha=ScalarParameter() {parameterName=Use_Trail}
  )
  n8 = Multiply(A=ScalarParameter() {defaultValue=14, parameterName=Spline}, B=DynamicParameter() {paramNames [Mask,Spline,Param3,Param4]}.Spline)
  n9 = MaterialFunctionCall(WidthBase=@n8, WidthTip=@n8) {materialFunction=&SplineThicken}
  n10 = OneMinus(None=ComponentMask(None=TextureCoordinate()) {g=true})
  n11 = Multiply(
   A=Multiply(
    A=Power(Base=@n10) {constExponent=4}
    B=DynamicParameter() {parameterIndex=1, defaultValue=<rgba 0 1 1 1>, paramNames [Gra,Param2,Param3,Param4]}.Gra
   )
   B=VectorParameter() {parameterName=Param_1, defaultValue=<rgba 0 0 -45 1>}.RGB
  )
  n12 = Constant3Vector() {constant=<rgba 1 0 0 0>}
  n13 = DynamicParameter() {paramNames [Mask,Spline,Param3,Param4]}
  n14 = Multiply(A=ComponentMask(None=TextureCoordinate()) {g=true}, B=Multiply(A=ScalarParameter() {defaultValue=2, parameterName=Wave}, B=@n13.Param3))
  n15 = Multiply(A=Time(), B=ScalarParameter() {defaultValue=1, parameterName=WPOSpeed})
  n16 = Multiply(A=Sine(None=Add(A=@n14, B=@n15)), B=Multiply(A=ScalarParameter() {parameterName=WPO}, B=@n13.Param4)) {constB=10}
  n17 = LinearInterpolate(A=@n16, B=Multiply(A=Transform(None=@n12), B=@n16), Alpha=ScalarParameter() {parameterName=NewDirection})
  LinearInterpolate(A=@n9.WorldPositionOffset, B=Add(A=@n9.WorldPositionOffset, B=Add(A=@n11, B=@n17)), Alpha=ScalarParameter() {parameterName=Use_Trail})
  n18 = TextureSample(UVs=Panner(Coordinate=TextureCoordinate() {uTiling=0.01, vTiling=4}) {speedY=0.2}) {texture=&T_ShieldCaustics}
  n19 = TextureSample(UVs=Panner(Coordinate=TextureCoordinate() {uTiling=0.05, vTiling=3}) {speedY=-0.3}) {texture=&T_ShieldCaustics}
  n20 = Multiply(A=@n1, B=@n1)
  n21 = TextureSample(UVs=Panner(Coordinate=TextureCoordinate() {uTiling=0.1, vTiling=3}) {speedX=0.2}) {samplerType=SAMPLERTYPE_Grayscale, texture=&LowResBlurredNoise}
  n22 = Add(A=Multiply(A=Multiply(A=@n18.R, B=@n19.R)) {constB=0.35}, B=Multiply(A=@n1, B=Add(A=@n20, B=@n21.R)))
  n23 = VectorParameter() {parameterName=Color, defaultValue=<rgba 0.5 0.0495 0.4995 1>}
  n24 = ScalarParameter() {parameterName=Desaturation}
  MaterialFunctionCall(Emissive=LinearInterpolate(A=Multiply(A=@n22, B=@n23.RGB), B=Desaturation(None=@n4.RGB, Fraction=@n24), Alpha=ScalarParameter() {parameterName=Use_Trail})) {materialFunction=&MF_FX_EmissiveAdjustments}
 ]
}
