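# Niagara system for the Eclipse backpack cosmetic that plays a looping eclipse effect of masks, glows, flares and lights attached to the pack's skeletal mesh.
# Timed emitters share an 18-second loop, bursting twice per cycle, with spawn groups selecting each burst's offsets.
# Two staggered Mask mesh emitters animate dynamic material parameters with mirrored curves, while persistent and continuously spawned meshes supply the base look.
# Emitters Empty004 and Glow002 are disabled.
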NiagaraSystem /BRCosmetics/Accessories/FORT_Backpacks/Backpack_F_MED_Eclipse/FX/NS_Backpack_Eclipse
dependencies {
 usedDataInterfaces [
  &NiagaraDataInterfaceSkeletalMesh
  &NiagaraDataInterfaceCurve
  &NiagaraDataInterfaceArrayFloat3
  &NiagaraDataInterfaceColorCurve
  &NiagaraDataInterfaceVectorCurve
  &NiagaraDataInterfaceArrayFloat
 ]
 usedDynamicInputs [&VectorFromCurve]
 usedModules [&SystemState, &Timeline]
}
emitters {
 CameraFlare {
  emitterUpdateScript [
   &EmitterState(Life Cycle Mode=::NewEnumerator1, Loop Duration=18, UseLoopDelay=false)
   &SpawnBurst_Instantaneous(Spawn Time=3.5, Use Spawn Probability=false, UseLoopCountLimit=false)
   &SpawnBurst_Instantaneous(Spawn Group=1, Spawn Time=13.5, Use Spawn Probability=false, UseLoopCountLimit=false) {module=SpawnBurst_Instantaneous001}
   &SpawnRate(SpawnRate=4, Use Spawn Probability=false) {enabled=false}
  ]
  particleSpawnScript [
   &FN_SetSpawnGroup(Number of Spawn Groups=2)
   &InitializeParticle(
    Color=<rgba 6 2.125 5.4997 1>
    Lifetime=4.6
    Mesh Scale=<2.2 2.2 2>
    Mesh Scale Mode=::NewEnumerator4
    Sprite Rotation Angle=&SelectFloatFromArray(Direct Array Index=$Particles.SpawnGroup, Float Selection Array=DataInterface() {propertyValues {FloatData [0,0.52]}})
    Sprite Rotation Mode=::NewEnumerator1
    Sprite Size Mode=::NewEnumerator3
    Uniform Sprite Size=35
    UsePositionOffset=false
    Write Material Random=true
   )
   &InitialMeshOrientation(
    Mesh Orientation Mode=::NewEnumerator7
    Rotation=&SelectVectorFromArray(Direct Array Index=$Particles.SpawnGroup, Vector Selection Array=DataInterface() {propertyValues {FloatData [<0 -0.38 0>, <0 0.14 0>]}})
    Use Rotation Vector=true
   )
   &InitialMeshOrientation(
    Mesh Orientation Mode=::NewEnumerator7
    Rotation=&SelectVectorFromArray(Direct Array Index=$Particles.SpawnGroup, Vector Selection Array=DataInterface() {propertyValues {FloatData [<0 0 0.04>, <0 0 -0.04>]}})
    Use Rotation Vector=true
   ) {module=InitialMeshOrientation001}
  ]
  particleUpdateScript [
   &ParticleState()
   &SkeletalMeshLocation(
    Bone / Socket Index=&SpawnGroup()
    Bone Sampling Mode=::NewEnumerator4
    MaskSpawnGroup=false
    Sampled Position Offset=&SelectVectorFromArray(
     Direct Array Index=&SpawnGroup()
     Vector Selection Array=DataInterface() {propertyValues {FloatData [<-7 -10 -7>, <5 -10 6>]}}
    )
    Skeletal Mesh=DataInterface() {propertyValues {
      DefaultMesh=null
      ExcludeBoneName=root
      OptionalComponentClass=null
      PreviewMesh=/BRCosmetics/Accessories/FORT_Backpacks/Backpack_F_MED_Eclipse/Meshes/F_MED_Eclipse_Pack
      SkinningMode=SkinOnTheFly
      SoftSourceActor=null
      SourceMode=Default
      UvSetIndex=0
      WholeMeshLOD=-1
      bExcludeBone=false
      bReadDeformedGeometry=true
      bRequireCurrentFrameData=true
      ComponentTags []
      FilteredBones [root]
      FilteredSockets []
      MeshUserParameter {parameter {name=None, typeDefHandle {registeredTypeIndex=1}, varData []}}
      SamplingRegions []
     }}
   )
   &ScaleColor(
    Scale Alpha=&FloatFromCurve(FloatCurve=DataInterface() {propertyValues {Curve {keys [
         {arriveTangent=3.9477, interpMode=RCIM_Cubic, leaveTangent=3.9477, tangentMode=RCTM_User, time=0, value=0}
         {arriveTangent=-0.246, interpMode=RCIM_Cubic, leaveTangent=-2.1933, tangentMode=RCTM_Break, time=0.5, value=0.9}
         {arriveTangent=-0.5243, interpMode=RCIM_Cubic, leaveTangent=-0.5243, tangentMode=RCTM_User, time=1, value=0}
        ]}}})
    ScaleA=true
    ScaleRGB=true
   )
   &ScaleMeshSize(Scale Factor=&VectorFromFloat(Value=&FloatFromCurve(FloatCurve=DataInterface() {propertyValues {Curve {keys [
          {arriveTangent=1.3203, interpMode=RCIM_Cubic, leaveTangent=1.3203, tangentMode=RCTM_User, time=0, value=1}
          {arriveTangent=0.0041, interpMode=RCIM_Cubic, leaveTangent=0.0041, tangentMode=RCTM_User, time=1, value=1.8}
         ]}}})))
  ]
  rendererClasses [&NiagaraMeshRendererProperties]
  renderers [{rendererClass=&NiagaraMeshRendererProperties, rendererIndex=0}]
 }
 Empty {
  emitterUpdateScript [
   &EmitterState()
   &SpawnRate(SpawnRate=10, Use Spawn Probability=false)
  ]
  particleSpawnScript [
   &InitializeParticle(
    Color=<rgba 4 4 4 1>
    Lifetime Mode=::NewEnumerator1
    Mesh Scale Mode=::NewEnumerator3
    Mesh Uniform Scale=1.3
    UsePositionOffset=false
    Write Material Random=true
   )
   &InitialMeshOrientation(Mesh Orientation Mode=::NewEnumerator7, Rotation=<0.3 0 0>, Use Rotation Vector=true)
   &InitialMeshOrientation(
    Mesh Orientation Mode=::NewEnumerator7
    Rotation=&RandomRangeVector(Minimum=<0 0 0>, Maximum=<0 1 0>)
    Use Rotation Vector=true
   ) {module=InitialMeshOrientation001}
  ]
  particleUpdateScript [
   &ParticleState()
   &SkeletalMeshLocation(
    Bone Sampling Mode=::NewEnumerator4
    MaskSpawnGroup=false
    Sampled Position Offset=<0 -8 -1>
    Skeletal Mesh=DataInterface() {propertyValues {
      DefaultMesh=null
      ExcludeBoneName=root
      OptionalComponentClass=null
      PreviewMesh=&F_MED_Eclipse_Pack
      SkinningMode=SkinOnTheFly
      SoftSourceActor=null
      SourceMode=Default
      UvSetIndex=0
      WholeMeshLOD=-1
      bExcludeBone=false
      bReadDeformedGeometry=true
      bRequireCurrentFrameData=true
      ComponentTags []
      FilteredBones [root]
      FilteredSockets []
      MeshUserParameter {parameter {name=None, typeDefHandle {registeredTypeIndex=1}, varData []}}
      SamplingRegions []
     }}
   )
   &ScaleColor(
    Linear Color Curve=DataInterface() {propertyValues {
      AlphaCurve {keys [<at 0 0>,<at 0.3383 1>,<at 1 0>]}
      BlueCurve {keys [<at 0 1>,<at 1 1>]}
      GreenCurve {keys [<at 0 0.4456>,<at 1 0.2682>]}
      RedCurve {keys [<at 0 0.9594>,<at 1 0.6938>]}
     }}
    Scale Mode=::NewEnumerator2
   )
   &ScaleMeshSize(Scale Factor=&Multiply_Vector(A=<1 1 1>, B=$Engine.Owner.Scale))
  ]
  rendererClasses [&NiagaraMeshRendererProperties]
  renderers [{rendererClass=&NiagaraMeshRendererProperties, rendererIndex=0}]
 }
 Empty001 {
  emitterUpdateScript [
   &EmitterState(Life Cycle Mode=::NewEnumerator1, Loop Behavior=::NewEnumerator1, Loop Duration Mode=::NewEnumerator1, UseLoopDelay=false)
   &SpawnRate(SpawnRate=10, Use Spawn Probability=false) {enabled=false}
   &SpawnBurst_Instantaneous(Use Spawn Probability=false, UseLoopCountLimit=false)
  ]
  particleSpawnScript [
   &InitializeParticle(
    Color=<rgba 1 0.9479 0.9928 0.15>
    Lifetime Mode=::NewEnumerator1
    Mesh Scale Mode=::NewEnumerator3
    Mesh Uniform Scale=0.4
    UsePositionOffset=false
    Write Material Random=true
   )
   &InitialMeshOrientation(Mesh Orientation Mode=::NewEnumerator7, Rotation=<0 0 0.25>, Use Rotation Vector=true)
   &InitialMeshOrientation(
    Mesh Orientation Mode=::NewEnumerator7
    Rotation=&RandomRangeVector(Minimum=<0 0 0>, Maximum=<0 1 0>)
    Use Rotation Vector=true
   ) {enabled=false, module=InitialMeshOrientation001}
  ]
  particleUpdateScript [
   &ParticleState(Kill Particles When Lifetime Has Elapsed=false, Let Infinitely Lived Particles Die When Emitter Deactivates=true)
   &SkeletalMeshLocation(
    Bone Sampling Mode=::NewEnumerator4
    MaskSpawnGroup=false
    Sampled Position Offset=<0 -10 -1>
    Skeletal Mesh=DataInterface() {propertyValues {
      DefaultMesh=null
      ExcludeBoneName=root
      OptionalComponentClass=null
      PreviewMesh=&F_MED_Eclipse_Pack
      SkinningMode=SkinOnTheFly
      SoftSourceActor=null
      SourceMode=Default
      UvSetIndex=0
      WholeMeshLOD=-1
      bExcludeBone=false
      bReadDeformedGeometry=true
      bRequireCurrentFrameData=true
      ComponentTags []
      FilteredBones [root]
      FilteredSockets []
      MeshUserParameter {parameter {name=None, typeDefHandle {registeredTypeIndex=1}, varData []}}
      SamplingRegions []
     }}
   )
   &ScaleColor(
    Linear Color Curve=DataInterface() {propertyValues {
      AlphaCurve {keys [<at 0 0>,<at 0.1013 0.3>,<at 1 0>]}
      BlueCurve {keys [<at 0 1>,<at 1 1>]}
      GreenCurve {keys [<at 0 0.4456>,<at 1 0.2682>]}
      RedCurve {keys [<at 0 0.9594>,<at 1 0.6938>]}
     }}
    Scale Mode=::NewEnumerator2
   ) {enabled=false}
  ]
  rendererClasses [&NiagaraMeshRendererProperties]
  renderers [{rendererClass=&NiagaraMeshRendererProperties, rendererIndex=0}]
 }
 Empty002 {
  emitterUpdateScript [
   &EmitterState(Life Cycle Mode=::NewEnumerator1, Loop Duration=18, UseLoopDelay=false)
   &SpawnBurst_Instantaneous(Spawn Time=3.7, Use Spawn Probability=false, UseLoopCountLimit=false)
   &SpawnBurst_Instantaneous(Spawn Group=1, Spawn Time=13.4, Use Spawn Probability=false, UseLoopCountLimit=false) {module=SpawnBurst_Instantaneous001}
   &SpawnRate(SpawnRate=10, Use Spawn Probability=false) {enabled=false}
  ]
  particleSpawnScript [
   &InitializeParticle(
    Color=<rgba 1 1 1 0.34>
    Lifetime=5
    Sprite Size Mode=::NewEnumerator3
    Uniform Sprite Size=28
    UsePositionOffset=false
    Write Material Random=true
   )
   &FN_SetSpawnGroup(Number of Spawn Groups=2)
   &CameraOffset()
  ]
  particleUpdateScript [
   &ParticleState()
   &SkeletalMeshLocation(
    Bone / Socket Index=$Particles.SpawnGroup
    Bone Sampling Mode=::NewEnumerator4
    MaskSpawnGroup=false
    Sampled Position Offset=&SelectVectorFromArray(
     Direct Array Index=&SpawnGroup()
     Vector Selection Array=DataInterface() {propertyValues {FloatData [<-8 -10 -6>, <7 -10 7>]}}
    )
    Skeletal Mesh=DataInterface() {propertyValues {
      DefaultMesh=null
      ExcludeBoneName=root
      OptionalComponentClass=null
      PreviewMesh=&F_MED_Eclipse_Pack
      SkinningMode=SkinOnTheFly
      SoftSourceActor=null
      SourceMode=Default
      UvSetIndex=0
      WholeMeshLOD=-1
      bExcludeBone=false
      bReadDeformedGeometry=true
      bRequireCurrentFrameData=true
      ComponentTags []
      FilteredBones [root]
      FilteredSockets []
      MeshUserParameter {parameter {name=None, typeDefHandle {registeredTypeIndex=1}, varData []}}
      SamplingRegions []
     }}
   )
   &ScaleSpriteSize(Uniform Curve Sprite Scale=DataInterface() {propertyValues {Curve {keys [
        {interpMode=RCIM_Cubic, leaveTangent=1.0109, tangentMode=RCTM_Break, time=0, value=1}
        {interpMode=RCIM_Cubic, tangentMode=RCTM_Break, time=1, value=1.4}
       ]}}})
   &ScaleColor(
    Scale Alpha=&FloatFromCurve(FloatCurve=DataInterface() {propertyValues {Curve {keys [
         {arriveTangent=7.3207, interpMode=RCIM_Cubic, leaveTangent=7.3207, tangentMode=RCTM_User, time=0, value=0}
         {arriveTangent=-1.1651, interpMode=RCIM_Cubic, leaveTangent=-1.1651, tangentMode=RCTM_User, time=0.1408, value=0.6}
         {arriveTangent=-0.0405, interpMode=RCIM_Cubic, leaveTangent=-0.0405, tangentMode=RCTM_User, time=1, value=0}
        ]}}})
    ScaleA=true
    ScaleRGB=true
   )
  ]
  rendererClasses [&NiagaraSpriteRendererProperties]
  renderers [{rendererClass=&NiagaraSpriteRendererProperties}]
 }
 Empty003 {
  emitterUpdateScript [
   &EmitterState(Life Cycle Mode=::NewEnumerator1, Loop Duration=18, UseLoopDelay=false)
   &SpawnBurst_Instantaneous(Spawn Time=3.7, Use Spawn Probability=false, UseLoopCountLimit=false)
   &SpawnBurst_Instantaneous(Spawn Group=1, Spawn Time=13.4, Use Spawn Probability=false, UseLoopCountLimit=false) {module=SpawnBurst_Instantaneous001}
   &SpawnRate(SpawnRate=10, Use Spawn Probability=false) {enabled=false}
  ]
  particleSpawnScript [
   &InitializeParticle(
    Color=<rgba 1.25 1.25 1.25 1>
    Lifetime=4
    Mesh Scale Mode=::NewEnumerator1
    Mesh Uniform Scale Max=0.4
    Mesh Uniform Scale Min=0.4
    Sprite Size Mode=::NewEnumerator3
    Uniform Sprite Size=32
    UsePositionOffset=false
    Write Material Random=true
   )
   &FN_SetSpawnGroup(Number of Spawn Groups=2)
   &InitialMeshOrientation(
    Mesh Orientation Mode=::NewEnumerator7
    Rotation=&RandomRangeVector(Minimum=<0 0 0>, Maximum=<1 0 0>)
    Use Rotation Vector=true
   )
  ]
  particleUpdateScript [
   &ParticleState()
   &SkeletalMeshLocation(
    Bone / Socket Index=$Particles.SpawnGroup
    Bone Sampling Mode=::NewEnumerator4
    MaskSpawnGroup=false
    Orientation Sampling=::NewEnumerator1
    Sampled Orientation Coordinate Space=::Local
    Sampled Position Offset=&SelectVectorFromArray(
     Direct Array Index=&SpawnGroup()
     Vector Selection Array=DataInterface() {propertyValues {FloatData [<-8 -10 -6>, <7 -10 7>]}}
    )
    Skeletal Mesh=DataInterface() {propertyValues {
      DefaultMesh=null
      ExcludeBoneName=root
      OptionalComponentClass=null
      PreviewMesh=&F_MED_Eclipse_Pack
      SkinningMode=SkinOnTheFly
      SoftSourceActor=null
      SourceMode=Default
      UvSetIndex=0
      WholeMeshLOD=-1
      bExcludeBone=false
      bReadDeformedGeometry=true
      bRequireCurrentFrameData=true
      ComponentTags []
      FilteredBones [root]
      FilteredSockets []
      MeshUserParameter {parameter {name=None, typeDefHandle {registeredTypeIndex=1}, varData []}}
      SamplingRegions []
     }}
   )
   /SetVariables(
    Particles.GlowColor=&ColorFromCurve(ColorCurve=DataInterface() {propertyValues {
       AlphaCurve {keys [
         {arriveTangent=2.797, interpMode=RCIM_Cubic, leaveTangent=2.797, time=0, value=0}
         <at 0.1073 0.3>
         {arriveTangent=-0.336, interpMode=RCIM_Cubic, leaveTangent=-0.336, time=1, value=0}
        ]}
       BlueCurve {keys [
         {arriveTangent=-0.0055, interpMode=RCIM_Cubic, leaveTangent=-0.0055, time=0.4494, value=1}
         {arriveTangent=-0.0055, interpMode=RCIM_Cubic, leaveTangent=-0.0055, time=1, value=0.9945}
        ]}
       GreenCurve {keys [<at 0.4494 0 RCIM_Cubic>,<at 1 0 RCIM_Cubic>]}
       RedCurve {keys [
         {arriveTangent=1, interpMode=RCIM_Cubic, leaveTangent=1, time=0.4494, value=0}
         {arriveTangent=1, interpMode=RCIM_Cubic, leaveTangent=1, time=1, value=1}
        ]}
      }})
    Particles.GlowColor2:LinearColor=&ColorFromCurve(ColorCurve=DataInterface() {propertyValues {
       AlphaCurve {keys [
         {arriveTangent=2.797, interpMode=RCIM_Cubic, leaveTangent=2.797, time=0, value=0}
         <at 0.1073 0.3>
         {arriveTangent=-0.336, interpMode=RCIM_Cubic, leaveTangent=-0.336, time=1, value=0}
        ]}
       BlueCurve {keys [
         {arriveTangent=0.9945, interpMode=RCIM_Cubic, leaveTangent=0.9945, time=0.4549, value=0}
         {arriveTangent=0.9945, interpMode=RCIM_Cubic, leaveTangent=0.9945, time=1, value=0.9945}
        ]}
       GreenCurve {keys [
         {arriveTangent=-1, interpMode=RCIM_Cubic, leaveTangent=-1, time=0.4549, value=1}
         {arriveTangent=-1, interpMode=RCIM_Cubic, leaveTangent=-1, time=1, value=0}
        ]}
       RedCurve {keys [
         {arriveTangent=1, interpMode=RCIM_Cubic, leaveTangent=1, time=0.4549, value=0}
         {arriveTangent=1, interpMode=RCIM_Cubic, leaveTangent=1, time=1, value=1}
        ]}
      }})
   ) {isSetParameters=true}
   &ScaleSpriteSize(Uniform Curve Sprite Scale=DataInterface() {propertyValues {Curve {keys [
        {interpMode=RCIM_Cubic, leaveTangent=1.0109, tangentMode=RCTM_Break, time=0, value=1}
        {interpMode=RCIM_Cubic, tangentMode=RCTM_Break, time=1, value=1.4}
       ]}}})
   &ScaleColor(
    Scale Alpha=&FloatFromCurve(FloatCurve=DataInterface() {propertyValues {Curve {keys [
         {arriveTangent=9.0725, interpMode=RCIM_Cubic, leaveTangent=9.0725, tangentMode=RCTM_User, time=0, value=0}
         {arriveTangent=-0.8185, interpMode=RCIM_Cubic, leaveTangent=-0.8185, tangentMode=RCTM_User, time=0.3, value=1.05}
         {arriveTangent=-0.0405, interpMode=RCIM_Cubic, leaveTangent=-0.0405, tangentMode=RCTM_User, time=1, value=0}
        ]}}})
    Scale RGB=&MakeVectorFromLinearColorRGB(Color=&ColorFromCurve(ColorCurve=DataInterface() {propertyValues {
        AlphaCurve {keys [<at 0 1 RCIM_Cubic>,<at 1 1 RCIM_Cubic>]}
        BlueCurve {keys [
          {arriveTangent=0.7299, interpMode=RCIM_Cubic, leaveTangent=0.7299, time=0, value=0.5997}
          {arriveTangent=0.7299, interpMode=RCIM_Cubic, leaveTangent=0.7299, time=0.5485, value=1}
         ]}
        GreenCurve {keys [
          {arriveTangent=0.0016, interpMode=RCIM_Cubic, leaveTangent=0.0016, time=0, value=0.0493}
          {arriveTangent=0.0016, interpMode=RCIM_Cubic, leaveTangent=0.0016, time=0.5485, value=0.0502}
         ]}
        RedCurve {keys [
          {arriveTangent=-1.4783, interpMode=RCIM_Cubic, leaveTangent=-1.4783, time=0, value=1}
          {arriveTangent=-1.4783, interpMode=RCIM_Cubic, leaveTangent=-1.4783, time=0.5485, value=0.1892}
         ]}
       }}))
    ScaleA=true
    ScaleRGB=true
   )
  ]
  rendererClasses [&NiagaraMeshRendererProperties]
  renderers [{rendererClass=&NiagaraMeshRendererProperties, rendererIndex=0}]
 }
 Empty004 {
  bEnabled=false
  emitterUpdateScript [
   &EmitterState()
   &SpawnRate(Use Spawn Probability=false)
  ]
  particleSpawnScript [
   &FN_SetSpawnGroup(Number of Spawn Groups=2)
   &InitializeParticle(
    Color Mode=::NewEnumerator0
    Lifetime=&SelectFloatFromArray(Direct Array Index=&SpawnGroup(), Float Selection Array=DataInterface() {propertyValues {FloatData [0.5,7]}})
    UsePositionOffset=false
    Write Material Random=true
   )
   /SetVariables(
    Particles.NPo:Vector3f=&Add_Vector(A=Linked() {linkedVariable {name=Particles.Position, type=NiagaraPosition}}, B=<50 0 0>)
    Particles.Nage=2
   ) {isSetParameters=true}
  ]
  particleUpdateScript [&ParticleState()]
  rendererClasses [&NiagaraSpriteRendererProperties]
  renderers [
   {rendererClass=&NiagaraSpriteRendererProperties}
   {rendererClass=&NiagaraSpriteRendererProperties, rendererIndex=1}
  ]
 }
 Glow {
  emitterUpdateScript [
   &EmitterState(Life Cycle Mode=::NewEnumerator1, Loop Duration=18, UseLoopDelay=false)
   &SpawnBurst_Instantaneous(Spawn Time=3.7, Use Spawn Probability=false, UseLoopCountLimit=false)
   &SpawnBurst_Instantaneous(Spawn Group=1, Spawn Time=13.4, Use Spawn Probability=false, UseLoopCountLimit=false) {module=SpawnBurst_Instantaneous001}
   &SpawnRate(SpawnRate=2, Use Spawn Probability=false) {enabled=false}
  ]
  particleSpawnScript [
   &InitializeParticle(
    Color=<rgba 2 2 2 0.1>
    Lifetime=4.8
    Position Mode=::NewEnumerator0
    Sprite Size=<50 50>
    Sprite Size Mode=::NewEnumerator4
    Write Material Random=true
   )
   &CameraOffset(Camera Offset Amount=2)
   &FN_SetSpawnGroup(Number of Spawn Groups=2)
   /SetVariables(Particles.GlowColor=<rgba 0 2 0 0.1>, Particles.GlowColor2:LinearColor=<rgba 0 0 2 0.1>) {isSetParameters=true, module=SetVariables_#1}
  ]
  particleUpdateScript [
   &ParticleState()
   &SkeletalMeshLocation(
    Bone / Socket Index=$Particles.SpawnGroup
    Bone Sampling Mode=::NewEnumerator4
    MaskSpawnGroup=false
    Sampled Position Offset=&SelectVectorFromArray(
     Direct Array Index=&SpawnGroup()
     Vector Selection Array=DataInterface() {propertyValues {FloatData [<-9 -10 -14>, <8 -10 2>]}}
    )
    Skeletal Mesh=DataInterface() {propertyValues {
      DefaultMesh=null
      ExcludeBoneName=root
      OptionalComponentClass=null
      PreviewMesh=&F_MED_Eclipse_Pack
      SkinningMode=SkinOnTheFly
      SoftSourceActor=null
      SourceMode=Default
      UvSetIndex=0
      WholeMeshLOD=-1
      bExcludeBone=false
      bReadDeformedGeometry=true
      bRequireCurrentFrameData=true
      ComponentTags []
      FilteredBones [pack]
      FilteredSockets []
      MeshUserParameter {parameter {name=None, typeDefHandle {registeredTypeIndex=1}, varData []}}
      SamplingRegions []
     }}
   )
   &ScaleSpriteSize(Uniform Curve Sprite Scale=DataInterface() {propertyValues {Curve {keys [
        {arriveTangent=1.2766, interpMode=RCIM_Cubic, leaveTangent=1.2766, tangentMode=RCTM_User, time=0.2, value=0.8}
        {arriveTangent=0.1572, interpMode=RCIM_Cubic, leaveTangent=0.1572, tangentMode=RCTM_User, time=1, value=1.2}
       ]}}})
   /SetVariables(
    Particles.GlowColor=&ColorFromCurve(ColorCurve=DataInterface() {propertyValues {
       AlphaCurve {keys [
         {arriveTangent=1.4184, interpMode=RCIM_Cubic, leaveTangent=1.4184, time=0, value=0}
         <at 0.1128 0.16>
         {arriveTangent=-0.1803, interpMode=RCIM_Cubic, leaveTangent=-0.1803, time=1, value=0}
        ]}
       BlueCurve {keys [
         {arriveTangent=0.9388, interpMode=RCIM_Cubic, leaveTangent=0.9388, time=0.1184, value=0}
         {arriveTangent=0.9388, interpMode=RCIM_Cubic, leaveTangent=0.9388, time=0.8673, value=0.7031}
        ]}
       GreenCurve {keys [
         {arriveTangent=-2.6459, interpMode=RCIM_Cubic, leaveTangent=-2.6459, time=0.1184, value=2}
         {arriveTangent=-2.6459, interpMode=RCIM_Cubic, leaveTangent=-2.6459, time=0.8673, value=0.0183}
        ]}
       RedCurve {keys [
         {arriveTangent=0.882, interpMode=RCIM_Cubic, leaveTangent=0.882, time=0.1184, value=0}
         {arriveTangent=0.882, interpMode=RCIM_Cubic, leaveTangent=0.882, time=0.8673, value=0.6606}
        ]}
      }})
    Particles.GlowColor2:LinearColor=&ColorFromCurve(ColorCurve=DataInterface() {propertyValues {
       AlphaCurve {keys [
         {arriveTangent=1.4184, interpMode=RCIM_Cubic, leaveTangent=1.4184, time=0, value=0}
         <at 0.1128 0.16>
         {arriveTangent=-0.1803, interpMode=RCIM_Cubic, leaveTangent=-0.1803, time=1, value=0}
        ]}
       BlueCurve {keys [
         {arriveTangent=-1.3932, interpMode=RCIM_Cubic, leaveTangent=-1.3932, time=0.1387, value=2}
         {arriveTangent=-1.3932, interpMode=RCIM_Cubic, leaveTangent=-1.3932, time=1, value=0.8}
        ]}
       GreenCurve {keys [
         {arriveTangent=-0.0359, interpMode=RCIM_Cubic, leaveTangent=-0.0359, time=0.1387, value=0.0729}
         {arriveTangent=-0.0359, interpMode=RCIM_Cubic, leaveTangent=-0.0359, time=1, value=0.042}
        ]}
       RedCurve {keys [
         {arriveTangent=0.9554, interpMode=RCIM_Cubic, leaveTangent=0.9554, time=0.1387, value=0.0729}
         {arriveTangent=0.9554, interpMode=RCIM_Cubic, leaveTangent=0.9554, time=1, value=0.8958}
        ]}
      }})
   ) {isSetParameters=true, module=SetVariables_#2}
   &ScaleColor(
    Linear Color Curve=DataInterface() {propertyValues {
      AlphaCurve {keys [<at 0 0>,<at 0.1085 1>,<at 0.3817 1>,<at 1 0>]}
      BlueCurve {keys [<at 0 1.6096>,<at 0.7291 1>]}
      GreenCurve {keys [<at 0 0.1571>,<at 0.7291 0.1385>]}
      RedCurve {keys [<at 0 2>,<at 0.7291 0.0104>]}
     }}
    Scale Mode=::NewEnumerator2
   ) {module=ScaleColor001}
  ]
  rendererClasses [&NiagaraSpriteRendererProperties]
  renderers [{rendererClass=&NiagaraSpriteRendererProperties}]
 }
 Glow001 {
  emitterUpdateScript [
   &EmitterState(Life Cycle Mode=::NewEnumerator1, Loop Duration=18, UseLoopDelay=false)
   &SpawnBurst_Instantaneous(Spawn Time=2.2, Use Spawn Probability=false, UseLoopCountLimit=false)
   &SpawnBurst_Instantaneous(Spawn Group=1, Spawn Time=12.5, Use Spawn Probability=false, UseLoopCountLimit=false) {module=SpawnBurst_Instantaneous001}
  ]
  particleSpawnScript [
   &InitializeParticle(
    Color=<rgba 5 0.7646 3.906 0.12>
    Lifetime=6
    Position Mode=::NewEnumerator0
    Sprite Size=<20 20>
    Sprite Size Mode=::NewEnumerator4
    Write Material Random=true
   )
   &CameraOffset()
   &FN_SetSpawnGroup(Number of Spawn Groups=2)
   /SetVariables(Particles.LightRadius=28) {isSetParameters=true}
   /SetVariables(Particles.LightExponent=3) {isSetParameters=true}
  ]
  particleUpdateScript [
   &ParticleState()
   /SetVariables(Particles.LightColor=&ColorFromCurve(ColorCurve=DataInterface() {propertyValues {
       AlphaCurve {keys [<at 0 1 RCIM_Cubic>,<at 1 0 RCIM_Cubic>]}
       BlueCurve {keys [<at 0 0>, <at 0.2053 12 RCIM_Cubic>, {arriveTangent=-15.0995, interpMode=RCIM_Cubic, leaveTangent=-15.0995, time=1, value=0}]}
       GreenCurve {keys [<at 0 0>, <at 0.2053 1.2108 RCIM_Cubic>, {arriveTangent=-1.5236, interpMode=RCIM_Cubic, leaveTangent=-1.5236, time=1, value=0}]}
       RedCurve {keys [<at 0 0>, <at 0.2053 11.4357 RCIM_Cubic>, {arriveTangent=-14.3894, interpMode=RCIM_Cubic, leaveTangent=-14.3894, time=1, value=0}]}
      }})) {isSetParameters=true}
   &SkeletalMeshLocation(
    Bone / Socket Index=$Particles.SpawnGroup
    Bone Sampling Mode=::NewEnumerator4
    MaskSpawnGroup=false
    Sampled Position Offset=&SelectVectorFromArray(
     Direct Array Index=&SpawnGroup()
     Vector Selection Array=DataInterface() {propertyValues {FloatData [<-9 -9 -14>, <9 -5 0>]}}
    )
    Skeletal Mesh=DataInterface() {propertyValues {
      DefaultMesh=null
      ExcludeBoneName=root
      OptionalComponentClass=null
      PreviewMesh=&F_MED_Eclipse_Pack
      SkinningMode=SkinOnTheFly
      SoftSourceActor=null
      SourceMode=Default
      UvSetIndex=0
      WholeMeshLOD=-1
      bExcludeBone=false
      bReadDeformedGeometry=true
      bRequireCurrentFrameData=true
      ComponentTags []
      FilteredBones [pack]
      FilteredSockets []
      MeshUserParameter {parameter {name=None, typeDefHandle {registeredTypeIndex=1}, varData []}}
      SamplingRegions []
     }}
   )
   &ScaleColor(
    Scale Alpha=&FloatFromCurve(FloatCurve=DataInterface() {propertyValues {Curve {keys [
         {arriveTangent=0.5605, interpMode=RCIM_Cubic, leaveTangent=0.5605, tangentMode=RCTM_User, time=0, value=0}
         {arriveTangent=2.4144, interpMode=RCIM_Cubic, leaveTangent=-3.0547, tangentMode=RCTM_Break, time=0.6, value=1.15}
         {arriveTangent=-1.1244, interpMode=RCIM_Cubic, leaveTangent=-1.1244, tangentMode=RCTM_User, time=1, value=0}
        ]}}})
    ScaleA=true
    ScaleRGB=true
   ) {enabled=false}
   &ScaleSpriteSize(Uniform Curve Sprite Scale=DataInterface() {propertyValues {Curve {keys [
        {arriveTangent=-4.3058, interpMode=RCIM_Cubic, leaveTangent=-4.3058, tangentMode=RCTM_User, time=0.2, value=1}
        {arriveTangent=-0.4097, interpMode=RCIM_Cubic, leaveTangent=-0.4097, tangentMode=RCTM_User, time=1, value=0}
       ]}}})
  ]
  rendererClasses [&NiagaraLightRendererProperties]
  renderers [{rendererClass=&NiagaraLightRendererProperties, rendererIndex=0}]
 }
 Glow002 {
  bEnabled=false
  emitterUpdateScript [
   &EmitterState(Life Cycle Mode=::NewEnumerator1, Loop Duration=18, UseLoopDelay=false)
   &SpawnBurst_Instantaneous(Spawn Time=3.7, Use Spawn Probability=false, UseLoopCountLimit=false)
   &SpawnBurst_Instantaneous(Spawn Group=1, Spawn Time=13.4, Use Spawn Probability=false, UseLoopCountLimit=false) {module=SpawnBurst_Instantaneous001}
   &SpawnRate(SpawnRate=2, Use Spawn Probability=false) {enabled=false}
  ]
  particleSpawnScript [
   &InitializeParticle(
    Color=<rgba 2 2 2 0.1>
    Lifetime=4.8
    Position Mode=::NewEnumerator0
    Sprite Size=<50 50>
    Sprite Size Mode=::NewEnumerator4
    Write Material Random=true
   )
   &CameraOffset(Camera Offset Amount=2)
   &FN_SetSpawnGroup()
   /SetVariables(Particles.GlowColor=<rgba 0 2 0 0.1>, Particles.GlowColor2:LinearColor=<rgba 0 0 2 0.1>) {isSetParameters=true, module=SetVariables_#1}
   /SetVariables(Particles.RingColor:LinearColor=<rgba 1 1 1 0.34>, Particles.RingSize2:Vector2f=<28 28>) {isSetParameters=true}
  ]
  particleUpdateScript [
   &ParticleState()
   &SkeletalMeshLocation(
    Bone / Socket Index=$Particles.SpawnGroup
    Bone Sampling Mode=::NewEnumerator4
    MaskSpawnGroup=false
    Sampled Position Offset=&SelectVectorFromArray(
     Direct Array Index=&SpawnGroup()
     Vector Selection Array=DataInterface() {propertyValues {FloatData [<-9 -10 -14>, <8 -10 2>]}}
    )
    Skeletal Mesh=DataInterface() {propertyValues {
      DefaultMesh=null
      ExcludeBoneName=root
      OptionalComponentClass=null
      PreviewMesh=&F_MED_Eclipse_Pack
      SkinningMode=SkinOnTheFly
      SoftSourceActor=null
      SourceMode=Default
      UvSetIndex=0
      WholeMeshLOD=-1
      bExcludeBone=false
      bReadDeformedGeometry=true
      bRequireCurrentFrameData=true
      ComponentTags []
      FilteredBones [pack]
      FilteredSockets []
      MeshUserParameter {parameter {name=None, typeDefHandle {registeredTypeIndex=1}, varData []}}
      SamplingRegions []
     }}
   )
   &ScaleSpriteSize(Uniform Curve Sprite Scale=DataInterface() {propertyValues {Curve {keys [
        {arriveTangent=1.2766, interpMode=RCIM_Cubic, leaveTangent=1.2766, tangentMode=RCTM_User, time=0.2, value=0.8}
        {arriveTangent=0.1572, interpMode=RCIM_Cubic, leaveTangent=0.1572, tangentMode=RCTM_User, time=1, value=1.2}
       ]}}})
   /SetVariables(
    Particles.GlowColor=&ColorFromCurve(ColorCurve=DataInterface() {propertyValues {
       AlphaCurve {keys [
         {arriveTangent=1.4184, interpMode=RCIM_Cubic, leaveTangent=1.4184, time=0, value=0}
         <at 0.1128 0.16>
         {arriveTangent=-0.1803, interpMode=RCIM_Cubic, leaveTangent=-0.1803, time=1, value=0}
        ]}
       BlueCurve {keys [
         {arriveTangent=0.9388, interpMode=RCIM_Cubic, leaveTangent=0.9388, time=0.1184, value=0}
         {arriveTangent=0.9388, interpMode=RCIM_Cubic, leaveTangent=0.9388, time=0.8673, value=0.7031}
        ]}
       GreenCurve {keys [
         {arriveTangent=-2.6459, interpMode=RCIM_Cubic, leaveTangent=-2.6459, time=0.1184, value=2}
         {arriveTangent=-2.6459, interpMode=RCIM_Cubic, leaveTangent=-2.6459, time=0.8673, value=0.0183}
        ]}
       RedCurve {keys [
         {arriveTangent=0.882, interpMode=RCIM_Cubic, leaveTangent=0.882, time=0.1184, value=0}
         {arriveTangent=0.882, interpMode=RCIM_Cubic, leaveTangent=0.882, time=0.8673, value=0.6606}
        ]}
      }})
    Particles.GlowColor2:LinearColor=&ColorFromCurve(ColorCurve=DataInterface() {propertyValues {
       AlphaCurve {keys [
         {arriveTangent=1.4184, interpMode=RCIM_Cubic, leaveTangent=1.4184, time=0, value=0}
         <at 0.1128 0.16>
         {arriveTangent=-0.1803, interpMode=RCIM_Cubic, leaveTangent=-0.1803, time=1, value=0}
        ]}
       BlueCurve {keys [
         {arriveTangent=-1.3932, interpMode=RCIM_Cubic, leaveTangent=-1.3932, time=0.1387, value=2}
         {arriveTangent=-1.3932, interpMode=RCIM_Cubic, leaveTangent=-1.3932, time=1, value=0.8}
        ]}
       GreenCurve {keys [
         {arriveTangent=-0.0359, interpMode=RCIM_Cubic, leaveTangent=-0.0359, time=0.1387, value=0.0729}
         {arriveTangent=-0.0359, interpMode=RCIM_Cubic, leaveTangent=-0.0359, time=1, value=0.042}
        ]}
       RedCurve {keys [
         {arriveTangent=0.9554, interpMode=RCIM_Cubic, leaveTangent=0.9554, time=0.1387, value=0.0729}
         {arriveTangent=0.9554, interpMode=RCIM_Cubic, leaveTangent=0.9554, time=1, value=0.8958}
        ]}
      }})
   ) {isSetParameters=true, module=SetVariables_#2}
   &ScaleColor(
    Linear Color Curve=DataInterface() {propertyValues {
      AlphaCurve {keys [<at 0 0>,<at 0.1085 1>,<at 0.3817 1>,<at 1 0>]}
      BlueCurve {keys [<at 0 1.6096>,<at 0.7291 1>]}
      GreenCurve {keys [<at 0 0.1571>,<at 0.7291 0.1385>]}
      RedCurve {keys [<at 0 2>,<at 0.7291 0.0104>]}
     }}
    Scale Mode=::NewEnumerator2
   ) {module=ScaleColor001}
   /SetVariables(
    Particles.ColorUpdate:LinearColor=&MultiplyLinearColor(
     Linear Color=$Particles.RingColor
     Scale Factor=&MakeVector4_FromLinearColor(LinearColor=&MakeLinearColorFromVectorAndFloat(
       "Vector (RGB)"=<1 1 1>
       "Float (Alpha)"=&FloatFromCurve(FloatCurve=DataInterface() {propertyValues {Curve {keys [
            {arriveTangent=7.3207, interpMode=RCIM_Cubic, leaveTangent=7.3207, tangentMode=RCTM_User, time=0, value=0}
            {arriveTangent=-1.1651, interpMode=RCIM_Cubic, leaveTangent=-1.1651, tangentMode=RCTM_User, time=0.1408, value=0.6}
            {arriveTangent=-0.0405, interpMode=RCIM_Cubic, leaveTangent=-0.0405, tangentMode=RCTM_User, time=1, value=0}
           ]}}})
      ))
    )
    Particles.RingsizeUpdate:Vector2f=&Vector2DFromFloat(Value=&FloatFromCurve(
      FloatCurve=DataInterface() {propertyValues {Curve {keys [
          {interpMode=RCIM_Cubic, leaveTangent=1.0109, tangentMode=RCTM_Break, time=0, value=1}
          {interpMode=RCIM_Cubic, tangentMode=RCTM_Break, time=1, value=1.4}
         ]}}}
      Scale Curve=/Niagara/DynamicInputs/TypeConversions/Vector2DLength(Vector:Vector2f=$Particles.RingSize2)
     ))
   ) {isSetParameters=true}
  ]
  rendererClasses [&NiagaraSpriteRendererProperties]
  renderers [
   {rendererClass=&NiagaraSpriteRendererProperties}
   {rendererClass=&NiagaraSpriteRendererProperties, rendererIndex=1}
  ]
 }
 GlowMesh {
  emitterUpdateScript [
   &EmitterState(Life Cycle Mode=::NewEnumerator1, Loop Behavior=::NewEnumerator1, Loop Duration Mode=::NewEnumerator1, UseLoopDelay=false)
   &SpawnBurst_Instantaneous(Use Spawn Probability=false, UseLoopCountLimit=false)
  ]
  particleSpawnScript [
   &InitializeParticle(
    Color Mode=::NewEnumerator0
    Lifetime=5
    Mesh Scale Mode=::NewEnumerator3
    Mesh Uniform Scale=4
    UsePositionOffset=false
    Write Material Random=true
   )
   &InitialMeshOrientation(Mesh Orientation Mode=::NewEnumerator7, Use Rotation Vector=true)
  ]
  particleUpdateScript [
   &ParticleState(Kill Particles When Lifetime Has Elapsed=false, Let Infinitely Lived Particles Die When Emitter Deactivates=true)
   &SkeletalMeshLocation(
    Bone Sampling Mode=::NewEnumerator4
    MaskSpawnGroup=false
    Sampled Position Offset=<0 14 2>
    Skeletal Mesh=DataInterface() {propertyValues {
      DefaultMesh=null
      ExcludeBoneName=root
      OptionalComponentClass=null
      PreviewMesh=&F_MED_Eclipse_Pack
      SkinningMode=SkinOnTheFly
      SoftSourceActor=null
      SourceMode=Default
      UvSetIndex=0
      WholeMeshLOD=-1
      bExcludeBone=false
      bReadDeformedGeometry=true
      bRequireCurrentFrameData=true
      ComponentTags []
      FilteredBones [root]
      FilteredSockets []
      MeshUserParameter {parameter {name=None, typeDefHandle {registeredTypeIndex=1}, varData []}}
      SamplingRegions []
     }}
   )
  ]
  rendererClasses [&NiagaraMeshRendererProperties]
  renderers [{rendererClass=&NiagaraMeshRendererProperties, rendererIndex=0}]
 }
 Mask {
  emitterUpdateScript [
   &EmitterState(Inactive Response=::NewEnumerator1, Life Cycle Mode=::NewEnumerator1, Loop Duration=18, UseLoopDelay=false)
   &SpawnBurst_Instantaneous(Use Spawn Probability=false, UseLoopCountLimit=false)
  ]
  particleSpawnScript [
   &InitializeParticle(
    Color Mode=::NewEnumerator0
    Lifetime=15
    Mesh Scale Mode=::NewEnumerator3
    Mesh Uniform Scale=1.01
    UsePositionOffset=false
    Write Material Random=true
   )
   &InitialMeshOrientation(Mesh Orientation Mode=::NewEnumerator7, Rotation=<0 0 -0.25>, Use Rotation Vector=true)
   &InitialMeshOrientation(Mesh Orientation Mode=::NewEnumerator7, Rotation=<0 0.26 0>, Use Rotation Vector=true) {module=InitialMeshOrientation001}
  ]
  particleUpdateScript [
   &ParticleState()
   &SkeletalMeshLocation(
    Bone Sampling Mode=::NewEnumerator4
    MaskSpawnGroup=false
    Sampled Position Offset=<0 -9.6 -1>
    Skeletal Mesh=DataInterface() {propertyValues {
      DefaultMesh=null
      ExcludeBoneName=root
      OptionalComponentClass=null
      PreviewMesh=&F_MED_Eclipse_Pack
      SkinningMode=SkinOnTheFly
      SoftSourceActor=null
      SourceMode=Default
      UvSetIndex=0
      WholeMeshLOD=-1
      bExcludeBone=false
      bReadDeformedGeometry=true
      bRequireCurrentFrameData=true
      ComponentTags []
      FilteredBones [root]
      FilteredSockets []
      MeshUserParameter {parameter {name=None, typeDefHandle {registeredTypeIndex=1}, varData []}}
      SamplingRegions []
     }}
   )
   &DynamicMaterialParameters(
    Index 0 Param 1=&FloatFromCurve(FloatCurve=DataInterface() {propertyValues {Curve {keys [
         {interpMode=RCIM_Cubic, tangentMode=RCTM_User, time=0, value=0.2}
         {arriveTangent=1.8, interpMode=RCIM_Cubic, leaveTangent=1.8, time=0.7, value=0.475}
         {arriveTangent=11.9775, interpMode=RCIM_Cubic, leaveTangent=11.9775, tangentMode=RCTM_User, time=1, value=2}
        ]}}})
    Index 0 Param 2=&FloatFromCurve(FloatCurve=DataInterface() {propertyValues {Curve {keys [{arriveTangent=-0.352, interpMode=RCIM_Cubic, leaveTangent=-0.352, tangentMode=RCTM_User, time=0, value=0.5}, <at 0.25 0.44>, <at 1 0.5>]}}})
    Index 0 Param 3=&FloatFromCurve(FloatCurve=DataInterface() {propertyValues {Curve {keys [{arriveTangent=-0.352, interpMode=RCIM_Cubic, leaveTangent=-0.352, tangentMode=RCTM_User, time=0, value=0.5}, <at 0.25 0.44>, <at 1 0.5>]}}})
    Index 0 Param 4=0.5
    Module.Param0WriteEnabled=true
    Module.Param1WriteEnabled=true
    Module.Param2WriteEnabled=true
    Module.Param3WriteEnabled=true
   )
  ]
  rendererClasses [&NiagaraMeshRendererProperties]
  renderers [{rendererClass=&NiagaraMeshRendererProperties, rendererIndex=0}]
 }
 Mask2 {
  emitterUpdateScript [
   &EmitterState(Inactive Response=::NewEnumerator1, Life Cycle Mode=::NewEnumerator1, Loop Duration=18, UseLoopDelay=false)
   &SpawnBurst_Instantaneous(Spawn Time=10, Use Spawn Probability=false, UseLoopCountLimit=false)
  ]
  particleSpawnScript [
   &InitializeParticle(
    Color Mode=::NewEnumerator0
    Lifetime=15
    Mesh Scale Mode=::NewEnumerator3
    Mesh Uniform Scale=1.01
    UsePositionOffset=false
    Write Material Random=true
   )
   &InitialMeshOrientation(Mesh Orientation Mode=::NewEnumerator7, Rotation=<0 0 -0.25>, Use Rotation Vector=true)
   &InitialMeshOrientation(Mesh Orientation Mode=::NewEnumerator7, Rotation=<0 0.26 0>, Use Rotation Vector=true) {module=InitialMeshOrientation001}
  ]
  particleUpdateScript [
   &ParticleState()
   &SkeletalMeshLocation(
    Bone Sampling Mode=::NewEnumerator4
    MaskSpawnGroup=false
    Sampled Position Offset=<0 -9.6 -1>
    Skeletal Mesh=DataInterface() {propertyValues {
      DefaultMesh=null
      ExcludeBoneName=root
      OptionalComponentClass=null
      PreviewMesh=&F_MED_Eclipse_Pack
      SkinningMode=SkinOnTheFly
      SoftSourceActor=null
      SourceMode=Default
      UvSetIndex=0
      WholeMeshLOD=-1
      bExcludeBone=false
      bReadDeformedGeometry=true
      bRequireCurrentFrameData=true
      ComponentTags []
      FilteredBones [root]
      FilteredSockets []
      MeshUserParameter {parameter {name=None, typeDefHandle {registeredTypeIndex=1}, varData []}}
      SamplingRegions []
     }}
   )
   &DynamicMaterialParameters(
    Index 0 Param 1=&FloatFromCurve(FloatCurve=DataInterface() {propertyValues {Curve {keys [
         {interpMode=RCIM_Cubic, tangentMode=RCTM_User, time=0, value=0.2}
         {arriveTangent=1.8, interpMode=RCIM_Cubic, leaveTangent=1.8, time=0.7, value=0.475}
         {arriveTangent=11.9775, interpMode=RCIM_Cubic, leaveTangent=11.9775, tangentMode=RCTM_User, time=1, value=2}
        ]}}})
    Index 0 Param 2=&FloatFromCurve(FloatCurve=DataInterface() {propertyValues {Curve {keys [{arriveTangent=0.392, interpMode=RCIM_Cubic, leaveTangent=0.392, tangentMode=RCTM_User, time=0, value=0.5}, <at 0.25 0.56>, <at 1 0.5>]}}})
    Index 0 Param 3=&FloatFromCurve(FloatCurve=DataInterface() {propertyValues {Curve {keys [{arriveTangent=0.392, interpMode=RCIM_Cubic, leaveTangent=0.392, tangentMode=RCTM_User, time=0, value=0.5}, <at 0.25 0.56>, <at 1 0.5>]}}})
    Index 0 Param 4=0.5
    Module.Param0WriteEnabled=true
    Module.Param1WriteEnabled=true
    Module.Param2WriteEnabled=true
    Module.Param3WriteEnabled=true
   )
  ]
  rendererClasses [&NiagaraMeshRendererProperties]
  renderers [{rendererClass=&NiagaraMeshRendererProperties, rendererIndex=0}]
 }
 Shine {
  emitterUpdateScript [
   &EmitterState(Life Cycle Mode=::NewEnumerator1, Loop Duration=18, UseLoopDelay=false)
   &SpawnBurst_Instantaneous(Spawn Time=3, Use Spawn Probability=false, UseLoopCountLimit=false)
   &SpawnBurst_Instantaneous(Spawn Group=1, Spawn Time=13.2, Use Spawn Probability=false, UseLoopCountLimit=false) {module=SpawnBurst_Instantaneous001}
   &SpawnRate(Use Spawn Probability=false) {enabled=false}
   &SpawnRate(SpawnGroup=1, Use Spawn Probability=false) {enabled=false, module=SpawnRate001}
  ]
  particleSpawnScript [
   &InitializeParticle(
    Color=<rgba 5 4.0625 4.7656 1>
    Lifetime=4.8
    Mesh Scale=&SelectVectorFromArray(
     Direct Array Index=&SpawnGroup()
     Vector Selection Array=DataInterface() {propertyValues {FloatData [<2.8 2.8 2.8>, <2.4 2.8 2.8>]}}
    )
    Mesh Scale Mode=::NewEnumerator4
    UsePositionOffset=false
    Write Material Random=true
   )
   &FN_SetSpawnGroup(Number of Spawn Groups=2)
   &InitialMeshOrientation(
    Mesh Orientation Mode=::NewEnumerator7
    Rotation=&SelectVectorFromArray(Direct Array Index=$Particles.SpawnGroup, Vector Selection Array=DataInterface() {propertyValues {FloatData [<0 -0.38 0>, <0 0.1 0>]}})
    Use Rotation Vector=true
   )
   &CameraOffset(Camera Offset Amount=1)
  ]
  particleUpdateScript [
   &ParticleState()
   &SkeletalMeshLocation(
    Bone / Socket Index=$Particles.SpawnGroup
    Bone Sampling Mode=::NewEnumerator4
    MaskSpawnGroup=false
    Sampled Position Offset=&SelectVectorFromArray(
     Direct Array Index=&SpawnGroup()
     Vector Selection Array=DataInterface() {propertyValues {FloatData [<-0.5 0.2 -0.4>, <0 0.2 -0.9>]}}
    )
    Skeletal Mesh=DataInterface() {propertyValues {
      DefaultMesh=null
      ExcludeBoneName=root
      OptionalComponentClass=null
      PreviewMesh=&F_MED_Eclipse_Pack
      SkinningMode=SkinOnTheFly
      SoftSourceActor=null
      SourceMode=Default
      UvSetIndex=0
      WholeMeshLOD=-1
      bExcludeBone=false
      bReadDeformedGeometry=true
      bRequireCurrentFrameData=true
      ComponentTags []
      FilteredBones [root]
      FilteredSockets []
      MeshUserParameter {parameter {name=None, typeDefHandle {registeredTypeIndex=1}, varData []}}
      SamplingRegions []
     }}
   )
   &ScaleColor(
    Scale Alpha=&FloatFromCurve(FloatCurve=DataInterface() {propertyValues {Curve {keys [
         {arriveTangent=0.9229, interpMode=RCIM_Cubic, leaveTangent=0.9229, tangentMode=RCTM_User, time=0, value=0}
         {arriveTangent=-0.2238, interpMode=RCIM_Cubic, leaveTangent=-0.6601, tangentMode=RCTM_Break, time=0.6, value=1.15}
         {arriveTangent=-2.875, interpMode=RCIM_Cubic, leaveTangent=-2.875, time=0.8, value=0.4}
         {arriveTangent=0.1769, interpMode=RCIM_Cubic, leaveTangent=-4.5883, tangentMode=RCTM_Break, time=1, value=0}
        ]}}})
    ScaleA=true
    ScaleRGB=true
   )
   &DynamicMaterialParameters(
    Index 0 Param 1=&FloatFromCurve(FloatCurve=DataInterface() {propertyValues {Curve {keys [
         {arriveTangent=3.7305, interpMode=RCIM_Cubic, leaveTangent=3.7305, tangentMode=RCTM_User, time=0.68, value=0}
         {arriveTangent=2.5886, interpMode=RCIM_Cubic, leaveTangent=2.5886, tangentMode=RCTM_User, time=0.8, value=1}
         {arriveTangent=-6.7692, interpMode=RCIM_Cubic, leaveTangent=-6.7692, tangentMode=RCTM_User, time=0.9, value=0.7}
         {arriveTangent=-4.0816, interpMode=RCIM_Cubic, tangentMode=RCTM_Break, time=1, value=0}
        ]}}})
    Index 0 Param 2=&FloatFromCurve(FloatCurve=DataInterface() {propertyValues {Curve {keys [
         {arriveTangent=-2.7346, interpMode=RCIM_Cubic, leaveTangent=-2.7346, tangentMode=RCTM_User, time=0.5, value=1}
         {arriveTangent=-1.3079, interpMode=RCIM_Cubic, leaveTangent=-1.3079, tangentMode=RCTM_User, time=1, value=0}
        ]}}})
    Module.Param0WriteEnabled=true
    Module.Param1WriteEnabled=true
    Module.Param2WriteEnabled=true
    Module.Param3WriteEnabled=true
   )
  ]
  rendererClasses [&NiagaraMeshRendererProperties]
  renderers [{rendererClass=&NiagaraMeshRendererProperties, rendererIndex=0}]
 }
}
summary {
 systemName=NS_Backpack_Eclipse
 emitters [
  {emitterName=GlowMesh, rendererClasses [&NiagaraMeshRendererProperties]}
  {emitterName=Mask, rendererClasses [&NiagaraMeshRendererProperties]}
  {emitterName=Shine, rendererClasses [&NiagaraMeshRendererProperties]}
  {emitterName=Mask2, rendererClasses [&NiagaraMeshRendererProperties]}
  {emitterName=Glow, rendererClasses [&NiagaraSpriteRendererProperties]}
  {emitterName=Glow001, rendererClasses [&NiagaraLightRendererProperties]}
  {emitterName=Empty, rendererClasses [&NiagaraMeshRendererProperties]}
  {emitterName=Empty001, rendererClasses [&NiagaraMeshRendererProperties]}
  {emitterName=Empty002, rendererClasses [&NiagaraSpriteRendererProperties]}
  {emitterName=Empty003, rendererClasses [&NiagaraMeshRendererProperties]}
  {emitterName=CameraFlare, rendererClasses [&NiagaraMeshRendererProperties]}
  {bEnabled=false, emitterName=Empty004, rendererClasses [&NiagaraSpriteRendererProperties]}
  {bEnabled=false, emitterName=Glow002, rendererClasses [&NiagaraSpriteRendererProperties]}
 ]
 userVariables []
}
