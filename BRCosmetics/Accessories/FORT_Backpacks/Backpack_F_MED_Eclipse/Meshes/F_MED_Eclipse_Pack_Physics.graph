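# Physics asset for the Eclipse backpack cosmetic, modelling a small dangling charm attached to the pack.
# A sphere body covers the pack, and the three-bone charm uses two small capsules ending in a thin box.
# Constraints chain each charm bone to the one before it, starting at the pack, with matching swing limits and a looser twist limit on the end piece.
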PhysicsAsset /BRCosmetics/Accessories/FORT_Backpacks/Backpack_F_MED_Eclipse/Meshes/F_MED_Eclipse_Pack_Physics
bodies {
 charm_01 [{shapeType=Capsule, length=1.755, radius=0.505, center=<1.2257 -0.1926 -0.0025>, rotation=<rot 0 89.9549 -89.6979>}]
 charm_02 [{shapeType=Capsule, length=0.505, radius=0.505, center=<0.2902 0.0679 -0.0008>, rotation=<rot 0 87.8139 -89.7841>}]
 charm_03 [{shapeType=Box, extentX=0.5, extentY=3.51, extentZ=3.51, center=<2.2485 0.0038 0.311>, rotation=<rot 0 -91.3757 -89.1066>}]
 pack [{shapeType=Sphere, radius=12.3666, center=<0.2896 -4.6785 -7.1547>}]
}
constraints [
 {bone1Name=charm_01, bone2Name=pack, swing1LimitDegrees=25, swing2LimitDegrees=25, twistLimitDegrees=10}
 {bone1Name=charm_02, bone2Name=charm_01, swing1LimitDegrees=25, swing2LimitDegrees=25, twistLimitDegrees=10}
 {bone1Name=charm_03, bone2Name=charm_02, swing1LimitDegrees=25, swing2LimitDegrees=25, twistLimitDegrees=45}
]
